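# Base cloth-shaded anime character material for the SaltyBrim cosmetic, layering flipbook face and animated eyes into the shared 2023 anime shader.
# Head-to-camera direction or an override picks the mouth's side by flipping UV2, while the animation-driven mouth index selects side or front offsets.
# Static switches add nose-line flipping, sock handling for Kicks and the emote's Fresnel transform glow and afterglow in emissive.
# A temporal responsiveness output accompanies eye animation and pose changes.
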Material /BRCosmetics/Characters/Player/Male/Medium/Bodies/M_MED_SaltyBrim_Pin/Materials/M_SaltyBrim_Base
graph {
 expressions {
  MaterialExpressionAdd_0 {
   class=MaterialExpressionAdd
   inputs [
    A {from=MaterialExpressionOneMinus_2}
    B {from=MaterialExpressionDivide_0}
   ]
   properties {constB=0.2}
  }
  MaterialExpressionAdd_1 {
   class=MaterialExpressionAdd
   inputs [
    A {from=MaterialExpressionScalarParameter_19}
    B {from=MaterialExpressionScalarParameter_21}
   ]
  }
  MaterialExpressionAdd_2 {
   class=MaterialExpressionAdd
   inputs [
    A {from=MaterialExpressionSubtract_1}
    B {from=MaterialExpressionAppendVector_1}
   ]
  }
  MaterialExpressionAppendVector_0 {
   class=MaterialExpressionAppendVector
   inputs [
    A {from=MaterialExpressionVertexColor_0, output=""}
    B {from=MaterialExpressionVertexColor_0, output=""}
   ]
  }
  MaterialExpressionAppendVector_1 {
   class=MaterialExpressionAppendVector
   inputs [
    A {from=MaterialExpressionScalarParameter_2, output=""}
    B {from=MaterialExpressionScalarParameter_4, output=""}
   ]
  }
  MaterialExpressionAppendVector_16 {
   class=MaterialExpressionAppendVector
   inputs [
    A {from=MaterialExpressionLinearInterpolate_20, output=""}
    B {from=MaterialExpressionMaterialFunctionCall_27, output=G}
   ]
  }
  MaterialExpressionAppendVector_2 {
   class=MaterialExpressionAppendVector
   inputs [
    A {from=MaterialExpressionComponentMask_1, output=""}
    B {from=MaterialExpressionSubtract_3, output=""}
   ]
  }
  MaterialExpressionAppendVector_6 {
   class=MaterialExpressionAppendVector
   inputs [
    A {from=MaterialExpressionScalarParameter_7, output=""}
    B {from=MaterialExpressionScalarParameter_8, output=""}
   ]
  }
  MaterialExpressionChannelMaskParameter_0 {
   class=MaterialExpressionChannelMaskParameter
   inputs [None {from=MaterialExpressionStaticSwitchParameter_1, output=""}]
   properties {group=01_EyeAnimation_Parameters, parameterName=EyeMaskChannel, sortPriority=-32}
  }
  MaterialExpressionComponentMask_0 {class=MaterialExpressionComponentMask, inputs [None {from=MaterialExpressionTextureCoordinate_0}], properties {r=true}}
  MaterialExpressionComponentMask_1 {class=MaterialExpressionComponentMask, inputs [None {from=MaterialExpressionTextureCoordinate_1}], properties {r=true}}
  MaterialExpressionComponentMask_2 {class=MaterialExpressionComponentMask, inputs [None {from=MaterialExpressionTextureCoordinate_1}], properties {g=true}}
  MaterialExpressionComponentMask_7 {class=MaterialExpressionComponentMask, inputs [None {from=MaterialExpressionFrac_0}], properties {g=true, r=true}}
  MaterialExpressionConstant3Vector_0 {class=MaterialExpressionConstant3Vector, properties {constant=<rgba 1 1 0 1>}}
  MaterialExpressionConstant3Vector_1 {class=MaterialExpressionConstant3Vector, properties {constant=<rgba 1 1 0.5 1>}}
  MaterialExpressionConstantBiasScale_3 {class=MaterialExpressionConstantBiasScale, inputs [None {from=MaterialExpressionDotProduct_7}]}
  MaterialExpressionConstant_1 {class=MaterialExpressionConstant}
  MaterialExpressionConstant_2 {class=MaterialExpressionConstant, properties {r=1}}
  MaterialExpressionCustom_0 {
   class=MaterialExpressionCustom
   inputs [Index {from=MaterialExpressionScalarParameter_1, output=""}]
   properties {
    code="int i = (int)round(Index);\r\nreturn (i==1 || i==4 || i==7 || i==10 || i==11 || i==12 || i==15) ? 1.0 : 0.0;\r\n"
    outputType=CMOT_Float1
    showCode=true
    inputs [{
      inputName=Index
      input {
       expression=&self:MaterialExpressionScalarParameter_1
       inputName=None
       mask=0
       maskA=0
       maskB=0
       maskG=0
       maskR=0
       outputIndex=0
      }
     }]
   }
  }
  MaterialExpressionDivide_0 {
   class=MaterialExpressionDivide
   inputs [
    A {from=MaterialExpressionScalarParameter_16}
    B {from=""}
   ]
  }
  MaterialExpressionDivide_1 {
   class=MaterialExpressionDivide
   inputs [
    A {from=MaterialExpressionSubtract_4}
    B {from=MaterialExpressionScalarParameter_20}
   ]
  }
  MaterialExpressionDotProduct_7 {
   class=MaterialExpressionDotProduct
   inputs [
    A {from=MaterialExpressionVectorParameter_9, output=RGB}
    B {from=MaterialExpressionTransform_18, output=""}
   ]
   properties {}
  }
  MaterialExpressionDotProduct_8 {
   class=MaterialExpressionDotProduct
   inputs [
    A {from=MaterialExpressionVectorParameter_8, output=RGB}
    B {from=MaterialExpressionTransform_18, output=""}
   ]
   properties {}
  }
  MaterialExpressionFrac_0 {class=MaterialExpressionFrac, inputs [None {from=MaterialExpressionTextureCoordinate_9, output=""}], properties {}}
  MaterialExpressionFrac_1 {class=MaterialExpressionFrac, inputs [None {from=MaterialExpressionTextureCoordinate_2, output=""}], properties {}}
  MaterialExpressionFrac_2 {class=MaterialExpressionFrac, inputs [None {from=MaterialExpressionComponentMask_1, output=""}], properties {}}
  MaterialExpressionFresnel_0 {
   class=MaterialExpressionFresnel
   inputs [
    ExponentIn {from=MaterialExpressionScalarParameter_11, output=""}
    BaseReflectFractionIn {from=MaterialExpressionScalarParameter_16, output=""}
    Normal {from="", output=""}
   ]
   properties {baseReflectFraction=0.5, exponent=0.5}
  }
  MaterialExpressionFresnel_1 {
   class=MaterialExpressionFresnel
   inputs [
    ExponentIn {from=MaterialExpressionScalarParameter_23, output=""}
    BaseReflectFractionIn {from="", output=""}
    Normal {from="", output=""}
   ]
   properties {baseReflectFraction=0, exponent=2.5}
  }
  MaterialExpressionGetMaterialAttributes_0 {
   class=MaterialExpressionGetMaterialAttributes
   inputs [None {from=MaterialExpressionStaticSwitchParameter_2, output=""}]
   properties {attributeGetTypes [#1]}
  }
  MaterialExpressionIf_0 {
   class=MaterialExpressionIf
   inputs [
    A {from=MaterialExpressionScalarParameter_13, output=""}
    B {from="", output=""}
    "A > B" {from=MaterialExpressionConstant_1, output=""}
    "A == B" {from=MaterialExpressionConstant_2, output=""}
    "A < B" {from=MaterialExpressionSaturate_14, output=""}
   ]
   properties {constB=1}
  }
  MaterialExpressionLinearInterpolate_0 {
   class=MaterialExpressionLinearInterpolate
   inputs [
    A {from=MaterialExpressionSign_3}
    B {from=MaterialExpressionOneMinus_1}
    Alpha {from=MaterialExpressionScalarParameter_0}
   ]
  }
  MaterialExpressionLinearInterpolate_1 {
   class=MaterialExpressionLinearInterpolate
   inputs [
    A {from=MaterialExpressionMaterialFunctionCall_3, output=Result}
    B {from=MaterialExpressionAdd_2}
    Alpha {from=MaterialExpressionLinearInterpolate_7}
   ]
  }
  MaterialExpressionLinearInterpolate_19 {
   class=MaterialExpressionLinearInterpolate
   inputs [
    A {from=MaterialExpressionSaturate_13}
    B {from=MaterialExpressionConstantBiasScale_3}
    Alpha {from=MaterialExpressionSaturate_12}
   ]
  }
  MaterialExpressionLinearInterpolate_2 {
   class=MaterialExpressionLinearInterpolate
   inputs [
    A {from=MaterialExpressionLinearInterpolate_6}
    B {from=MaterialExpressionMultiply_9}
    Alpha {from=MaterialExpressionMultiply_10}
   ]
  }
  MaterialExpressionLinearInterpolate_20 {
   class=MaterialExpressionLinearInterpolate
   inputs [
    A {from=MaterialExpressionOneMinus_0}
    B {from=MaterialExpressionMaterialFunctionCall_27, output=R}
    Alpha {from=MaterialExpressionIf_0}
   ]
  }
  MaterialExpressionLinearInterpolate_3 {
   class=MaterialExpressionLinearInterpolate
   inputs [
    A {from=MaterialExpressionMaterialFunctionCall_21, output=Color_Lit}
    B {from=MaterialExpressionMaterialFunctionCall_11, output=FaceColor}
    Alpha {from=MaterialExpressionNamedRerouteDeclaration_1}
   ]
  }
  MaterialExpressionLinearInterpolate_4 {
   class=MaterialExpressionLinearInterpolate
   inputs [
    A {from=MaterialExpressionMaterialFunctionCall_21, output=Color_Shaded}
    B {from=MaterialExpressionMultiply_2}
    Alpha {from=MaterialExpressionNamedRerouteDeclaration_1}
   ]
  }
  MaterialExpressionLinearInterpolate_5 {
   class=MaterialExpressionLinearInterpolate
   inputs [
    A {from=""}
    B {from=""}
    Alpha {from=MaterialExpressionMultiply_3}
   ]
   properties {constA=0.4}
  }
  MaterialExpressionLinearInterpolate_6 {
   class=MaterialExpressionLinearInterpolate
   inputs [
    A {from=MaterialExpressionGetMaterialAttributes_0, output=Emissive Color}
    B {from=MaterialExpressionMultiply_5}
    Alpha {from=MaterialExpressionMultiply_6}
   ]
  }
  MaterialExpressionLinearInterpolate_7 {
   class=MaterialExpressionLinearInterpolate
   inputs [
    A {from=MaterialExpressionCustom_0}
    B {from=MaterialExpressionSwitch_0}
    Alpha {from=MaterialExpressionScalarParameter_9}
   ]
   properties {constAlpha=1}
  }
  MaterialExpressionMaterialFunctionCall_1 {
   class=MaterialExpressionMaterialFunctionCall
   inputs []
   properties {materialFunction=/Game/Global/Materials/Character_Parent_2023/Anime_Parent_2023/MaterialFunctions/MF_NoseLineFlip}
  }
  MaterialExpressionMaterialFunctionCall_11 {
   class=MaterialExpressionMaterialFunctionCall
   inputs [
    Mouth UV {from=MaterialExpressionLinearInterpolate_1}
    Eye UV {from=""}
    Brow UV {from=""}
    Material {from=""}
   ]
   properties {materialFunction=/Game/Global/Materials/Character_Parent_2023/FlipbookFace_Parent_2023/MaterialFunctions/MFI_FlipbookFace_Anime}
  }
  MaterialExpressionMaterialFunctionCall_14 {
   class=MaterialExpressionMaterialFunctionCall
   inputs [
    Shading_Gradient {from=MaterialExpressionMaterialFunctionCall_20, output=Shading_Gradient}
    Threshold_Gradient {from=MaterialExpressionMaterialFunctionCall_20, output=Shading_Gradient}
    Specular_Gradient {from=MaterialExpressionMaterialFunctionCall_20, output=Shading_Gradient}
    Sheen_Mask {from=MaterialExpressionMaterialFunctionCall_20, output=Shading_Gradient}
    Thickness {from=MaterialExpressionMaterialFunctionCall_20, output=Shading_Gradient}
    DFL_Mask {from=MaterialExpressionMaterialFunctionCall_20, output=Shading_Gradient}
    FX {from=MaterialExpressionMaterialFunctionCall_20, output=Shading_Gradient}
    Metallic_Gradient {from=MaterialExpressionMaterialFunctionCall_20, output=Shading_Gradient}
    RimLight_Gradient {from=MaterialExpressionMaterialFunctionCall_20, output=Shading_Gradient}
    NormalBuffer {from=MaterialExpressionMaterialFunctionCall_20, output=Shading_Gradient}
    -------- {from=""}
    CINE_LightSources {from=MaterialExpressionMaterialFunctionCall_20, output=Shading_Gradient}
    CINE_LightColor {from=MaterialExpressionMaterialFunctionCall_20, output=Shading_Gradient}
    --------- {from=""}
    Color_Lit {from=MaterialExpressionMaterialFunctionCall_4, output=Lit}
    Color_Shaded {from=MaterialExpressionMaterialFunctionCall_4, output=Lit}
    SRM {from=MaterialExpressionMaterialFunctionCall_21, output=SRM}
    Emissive {from=MaterialExpressionMaterialFunctionCall_21, output=SRM}
    DFL_Color {from=MaterialExpressionMaterialFunctionCall_21, output=SRM}
    Specular_Color {from=MaterialExpressionMaterialFunctionCall_21, output=SRM}
    RimLight_Color {from=MaterialExpressionMaterialFunctionCall_21, output=SRM}
   ]
   properties {materialFunction=/Game/Global/Materials/Character_Parent_2023/Anime_Parent_2023/MaterialFunctions/MFC_AnimeShader_2023_Post}
  }
  MaterialExpressionMaterialFunctionCall_2 {
   class=MaterialExpressionMaterialFunctionCall
   inputs [
    InOpacity {from=""}
    Material {from=MaterialExpressionMaterialFunctionCall_14, output=Result}
    Mask {from=""}
   ]
   properties {materialFunction=/Game/Global/Materials/Cosmetics/Kicks/Socks/Materials/MF_Sock}
  }
  MaterialExpressionMaterialFunctionCall_20 {
   class=MaterialExpressionMaterialFunctionCall
   inputs [
    CustomLightVector {from=""}
    OverrideNormal_WS {from=""}
    STT {from=MaterialExpressionMaterialFunctionCall_21, output=STT}
    Normal_Map {from=MaterialExpressionMaterialFunctionCall_21, output=STT}
    DFL {from=MaterialExpressionStaticSwitchParameter_10}
    FX {from=MaterialExpressionMaterialFunctionCall_21, output=STT}
   ]
   properties {materialFunction=/Game/Global/Materials/Character_Parent_2023/Anime_Parent_2023/MaterialFunctions/MFC_AnimeShader_2023_Calculations}
  }
  MaterialExpressionMaterialFunctionCall_21 {
   class=MaterialExpressionMaterialFunctionCall
   inputs [CustomUV {from=""}]
   properties {materialFunction=/Game/Global/Materials/Character_Parent_2023/Anime_Parent_2023/MaterialFunctions/MFC_AnimeShader_2023_Pre}
  }
  MaterialExpressionMaterialFunctionCall_27 {
   class=MaterialExpressionMaterialFunctionCall
   inputs [Float2 {from=MaterialExpressionComponentMask_7}]
   properties {materialFunction=&BreakOutFloat2Components}
  }
  MaterialExpressionMaterialFunctionCall_3 {
   class=MaterialExpressionMaterialFunctionCall
   inputs [
    RotationAmount {from=MaterialExpressionScalarParameter_10}
    Coords {from=MaterialExpressionAppendVector_16}
    Pivot {from=""}
   ]
   properties {materialFunction=/BaseMaterial/Materials/Functions/MF_Rotate2D}
  }
  MaterialExpressionMaterialFunctionCall_4 {
   class=MaterialExpressionMaterialFunctionCall
   inputs [
    Color Lit {from=MaterialExpressionLinearInterpolate_3}
    Color Shaded {from=MaterialExpressionLinearInterpolate_4}
    Use UV2 {from=MaterialExpressionStaticBoolParameter_0}
    OverrideUVs {from=MaterialExpressionAppendVector_2}
    Mask {from=MaterialExpressionStaticSwitchParameter_0}
   ]
   properties {materialFunction=&/MF_AnimeEyeAnimation_SaltyBrim}
  }
  MaterialExpressionMax_0 {
   class=MaterialExpressionMax
   inputs [
    A {from=MaterialExpressionLinearInterpolate_5}
    B {from=MaterialExpressionNamedRerouteUsage_0}
   ]
  }
  MaterialExpressionMultiply_0 {
   class=MaterialExpressionMultiply
   inputs [
    A {from=MaterialExpressionSubtract_2}
    B {from=MaterialExpressionSubtract_2}
   ]
  }
  MaterialExpressionMultiply_1 {
   class=MaterialExpressionMultiply
   inputs [
    A {from=MaterialExpressionMultiply_0}
    B {from=MaterialExpressionScalarParameter_6}
   ]
  }
  MaterialExpressionMultiply_10 {
   class=MaterialExpressionMultiply
   inputs [
    A {from=MaterialExpressionSmoothStep_2}
    B {from=MaterialExpressionScalarParameter_24}
   ]
  }
  MaterialExpressionMultiply_18 {
   class=MaterialExpressionMultiply
   inputs [
    A {from=MaterialExpressionTransformPosition_0}
    B {from=""}
   ]
   properties {constB=-1}
  }
  MaterialExpressionMultiply_2 {
   class=MaterialExpressionMultiply
   inputs [
    A {from=MaterialExpressionMaterialFunctionCall_11, output=FaceColor}
    B {from=MaterialExpressionScalarParameter_3}
   ]
  }
  MaterialExpressionMultiply_3 {
   class=MaterialExpressionMultiply
   inputs [
    A {from=MaterialExpressionScalarParameter_5}
    B {from=MaterialExpressionNamedRerouteUsage_1}
   ]
  }
  MaterialExpressionMultiply_4 {
   class=MaterialExpressionMultiply
   inputs [
    A {from=MaterialExpressionSaturate_2}
    B {from=MaterialExpressionSaturate_3}
   ]
  }
  MaterialExpressionMultiply_5 {
   class=MaterialExpressionMultiply
   inputs [
    A {from=MaterialExpressionConstant3Vector_0}
    B {from=MaterialExpressionScalarParameter_12}
   ]
  }
  MaterialExpressionMultiply_6 {
   class=MaterialExpressionMultiply
   inputs [
    A {from=MaterialExpressionSaturate_1}
    B {from=MaterialExpressionScalarParameter_18}
   ]
  }
  MaterialExpressionMultiply_7 {
   class=MaterialExpressionMultiply
   inputs [
    A {from=MaterialExpressionScalarParameter_19}
    B {from=MaterialExpressionScalarParameter_21}
   ]
  }
  MaterialExpressionMultiply_9 {
   class=MaterialExpressionMultiply
   inputs [
    A {from=MaterialExpressionConstant3Vector_1}
    B {from=MaterialExpressionScalarParameter_22}
   ]
  }
  MaterialExpressionNamedRerouteDeclaration_0 {
   class=MaterialExpressionNamedRerouteDeclaration
   inputs [None {from=MaterialExpressionMaterialFunctionCall_4, output=TemporalResponsiveness}]
   properties {name=EyeAnimation_TemporalResp, nodeColor=<rgba 1 0.9141 0 1>}
  }
  MaterialExpressionNamedRerouteDeclaration_1 {
   class=MaterialExpressionNamedRerouteDeclaration
   inputs [None {from=MaterialExpressionMaterialFunctionCall_11, output=FaceAlpha}]
   properties {name=FlipbookFaceAlpha, nodeColor=<rgba 0.8281 0 1 1>}
  }
  MaterialExpressionNamedRerouteDeclaration_2 {
   class=MaterialExpressionNamedRerouteDeclaration
   inputs [None {from=MaterialExpressionMultiply_4}]
   properties {name=Time, nodeColor=<rgba 0.0781 0 1 1>}
  }
  MaterialExpressionNamedRerouteDeclaration_3 {
   class=MaterialExpressionNamedRerouteDeclaration
   inputs [None {from=MaterialExpressionSquareRoot_0}]
   properties {name=SquaredTime, nodeColor=<rgba 0.2422 1 0 1>}
  }
  MaterialExpressionNamedRerouteDeclaration_30 {
   class=MaterialExpressionNamedRerouteDeclaration
   inputs [None {from=MaterialExpressionLinearInterpolate_0}]
   properties {name=LeftRightSign, nodeColor=<rgba 1 0 0.1172 1>}
  }
  MaterialExpressionNamedRerouteUsage_0 {class=MaterialExpressionNamedRerouteUsage}
  MaterialExpressionNamedRerouteUsage_1 {class=MaterialExpressionNamedRerouteUsage}
  MaterialExpressionNamedRerouteUsage_29 {class=MaterialExpressionNamedRerouteUsage}
  MaterialExpressionNormalize_1 {class=MaterialExpressionNormalize, inputs [VectorInput {from=MaterialExpressionMultiply_18, output=""}]}
  MaterialExpressionOneMinus_0 {class=MaterialExpressionOneMinus, inputs [None {from=MaterialExpressionMaterialFunctionCall_27, output=R}]}
  MaterialExpressionOneMinus_1 {class=MaterialExpressionOneMinus, inputs [None {from=MaterialExpressionSign_3}]}
  MaterialExpressionOneMinus_2 {class=MaterialExpressionOneMinus, inputs [None {from=MaterialExpressionSmoothStep_0}]}
  MaterialExpressionSaturate_0 {class=MaterialExpressionSaturate, inputs [None {from=MaterialExpressionSign_0}]}
  MaterialExpressionSaturate_1 {class=MaterialExpressionSaturate, inputs [None {from=MaterialExpressionAdd_0}]}
  MaterialExpressionSaturate_12 {class=MaterialExpressionSaturate, inputs [None {from=MaterialExpressionSign_5}]}
  MaterialExpressionSaturate_13 {class=MaterialExpressionSaturate, inputs [None {from=MaterialExpressionSign_4}]}
  MaterialExpressionSaturate_14 {class=MaterialExpressionSaturate, inputs [None {from=MaterialExpressionNamedRerouteUsage_29}]}
  MaterialExpressionSaturate_2 {class=MaterialExpressionSaturate, inputs [None {from=MaterialExpressionDivide_1}]}
  MaterialExpressionSaturate_3 {class=MaterialExpressionSaturate, inputs [None {from=MaterialExpressionMultiply_7}]}
  MaterialExpressionScalarParameter_0 {
   class=MaterialExpressionScalarParameter
   properties {controlType=Enumeration, enumeration=/Game/Global/Materials/Enums/E_Mat_OnOff, group=Flipbook 04 Mouth, parameterName=FlipLeftRightSign}
  }
  MaterialExpressionScalarParameter_1 {
   class=MaterialExpressionScalarParameter
   properties {
    desc="Driven by Animation Blueprint.  Do not set manually."
    group=Flipbook 05 Indices
    parameterName=flipbook_face_mouth_index
    sortPriority=20
   }
  }
  MaterialExpressionScalarParameter_10 {class=MaterialExpressionScalarParameter, properties {group=Flipbook 04 Mouth, parameterName=UVRotation}}
  MaterialExpressionScalarParameter_11 {class=MaterialExpressionScalarParameter, properties {defaultValue=0.5, group=SaltyBrim, parameterName=TransformGlowExponent}}
  MaterialExpressionScalarParameter_12 {class=MaterialExpressionScalarParameter, properties {defaultValue=3, group=SaltyBrim, parameterName=TransformGlowIntensity}}
  MaterialExpressionScalarParameter_13 {
   class=MaterialExpressionScalarParameter
   properties {
    desc="Use this to manually override the side the flipbook mouth appears.\r\n0 = The normal behavior, camera direction determines the side of the mouth.\r\n1 = Force the mouth to be on the LEFT side always.\r\n2 = Force the mouth to be on the RIGHT side always."
    group=SaltyBrim
    parameterName=OverrideMouthSide
   }
  }
  MaterialExpressionScalarParameter_16 {class=MaterialExpressionScalarParameter, properties {defaultValue=0.7, group=SaltyBrim, parameterName=TransformGlowBase}}
  MaterialExpressionScalarParameter_18 {class=MaterialExpressionScalarParameter, properties {group=SaltyBrim, parameterName=TransformGlow}}
  MaterialExpressionScalarParameter_19 {class=MaterialExpressionScalarParameter, properties {group=SaltyBrim, parameterName=AfterGlowTimestamp}}
  MaterialExpressionScalarParameter_2 {class=MaterialExpressionScalarParameter, properties {group=Flipbook 04 Mouth, parameterName=FB_Front_MouthUVOffsetX, sortPriority=16}}
  MaterialExpressionScalarParameter_20 {class=MaterialExpressionScalarParameter, properties {defaultValue=5, group=SaltyBrim, parameterName=AfterGlowDuration}}
  MaterialExpressionScalarParameter_21 {class=MaterialExpressionScalarParameter, properties {defaultValue=5, parameterName=AfterGlowDelay}}
  MaterialExpressionScalarParameter_22 {class=MaterialExpressionScalarParameter, properties {defaultValue=5, group=SaltyBrim, parameterName=AfterGlowIntensity}}
  MaterialExpressionScalarParameter_23 {class=MaterialExpressionScalarParameter, properties {defaultValue=2.5, group=SaltyBrim, parameterName=AfterGlowExponent}}
  MaterialExpressionScalarParameter_24 {class=MaterialExpressionScalarParameter, properties {group=SaltyBrim, parameterName=AfterGlow}}
  MaterialExpressionScalarParameter_3 {class=MaterialExpressionScalarParameter, properties {defaultValue=1, parameterName=FlipbookFaceAnimeShadedValue}}
  MaterialExpressionScalarParameter_4 {class=MaterialExpressionScalarParameter, properties {group=Flipbook 04 Mouth, parameterName=FB_Front_MouthUVOffsetY, sortPriority=17}}
  MaterialExpressionScalarParameter_5 {class=MaterialExpressionScalarParameter, properties {parameterName=HasPoseChanged}}
  MaterialExpressionScalarParameter_6 {class=MaterialExpressionScalarParameter, properties {group=01_EyeAnimation_Parameters, parameterName=EyeLidsArc}}
  MaterialExpressionScalarParameter_7 {class=MaterialExpressionScalarParameter, properties {group=Flipbook 04 Mouth, parameterName=FB_MouthUVOffsetX, sortPriority=16}}
  MaterialExpressionScalarParameter_8 {class=MaterialExpressionScalarParameter, properties {group=Flipbook 04 Mouth, parameterName=FB_MouthUVOffsetY, sortPriority=17}}
  MaterialExpressionScalarParameter_9 {
   class=MaterialExpressionScalarParameter
   properties {controlType=Enumeration, enumeration=&E_Mat_OnOff, parameterName=UseCustomMiddleMouthIndex}
  }
  MaterialExpressionSetMaterialAttributes_0 {
   class=MaterialExpressionSetMaterialAttributes
   inputs [
    MaterialAttributes {from=MaterialExpressionGetMaterialAttributes_0, output=MaterialAttributes}
    Emissive Color {from=MaterialExpressionLinearInterpolate_2, output=""}
   ]
   properties {attributeSetTypes [#1]}
  }
  MaterialExpressionSign_0 {class=MaterialExpressionSign, inputs [None {from=MaterialExpressionSubtract_0, output=""}], properties {}}
  MaterialExpressionSign_3 {class=MaterialExpressionSign, inputs [None {from=MaterialExpressionSubtract_6, output=""}], properties {}}
  MaterialExpressionSign_4 {class=MaterialExpressionSign, inputs [None {from=MaterialExpressionDotProduct_7, output=""}], properties {}}
  MaterialExpressionSign_5 {class=MaterialExpressionSign, inputs [None {from=MaterialExpressionDotProduct_8, output=""}], properties {}}
  MaterialExpressionSmoothStep_0 {
   class=MaterialExpressionSmoothStep
   inputs [
    Min {from="", output=""}
    Max {from="", output=""}
    Value {from=MaterialExpressionFresnel_0, output=""}
   ]
   properties {}
  }
  MaterialExpressionSmoothStep_2 {
   class=MaterialExpressionSmoothStep
   inputs [
    Min {from="", output=""}
    Max {from="", output=""}
    Value {from=MaterialExpressionFresnel_1, output=""}
   ]
   properties {}
  }
  MaterialExpressionSquareRoot_0 {class=MaterialExpressionSquareRoot, inputs [None {from=MaterialExpressionNamedRerouteDeclaration_2, output=""}], properties {}}
  MaterialExpressionStaticBoolParameter_0 {
   class=MaterialExpressionStaticBoolParameter
   inputs []
   properties {defaultValue=true, group=01_EyeAnimation_Parameters, parameterName=Use UV2 for Eye Animation, sortPriority=-30}
  }
  MaterialExpressionStaticSwitchParameter_0 {
   class=MaterialExpressionStaticSwitchParameter
   inputs [
    True {from=MaterialExpressionChannelMaskParameter_0, output=R}
    False {from=MaterialExpressionSaturate_0, output=""}
   ]
   properties {group=01_EyeAnimation_Parameters, parameterName=UseEyeMask, sortPriority=-33}
  }
  MaterialExpressionStaticSwitchParameter_1 {
   class=MaterialExpressionStaticSwitchParameter
   inputs [
    True {from=MaterialExpressionAppendVector_0, output=""}
    False {from=MaterialExpressionTextureSampleParameter2D_0, output=RGBA}
   ]
   properties {group=01_EyeAnimation_Parameters, parameterName=UseVertexColorForEyeMask, sortPriority=-33}
  }
  MaterialExpressionStaticSwitchParameter_10 {
   class=MaterialExpressionStaticSwitchParameter
   inputs [
    True {from=MaterialExpressionMaterialFunctionCall_1, output=DFL}
    False {from=MaterialExpressionMaterialFunctionCall_21, output=DFL}
   ]
   properties {group=Nose, parameterName=UseNoseLineFlip, sortPriority=0}
  }
  MaterialExpressionStaticSwitchParameter_2 {
   class=MaterialExpressionStaticSwitchParameter
   inputs [
    True {from=MaterialExpressionMaterialFunctionCall_2, output=Result}
    False {from=MaterialExpressionMaterialFunctionCall_14, output=Result}
   ]
   properties {group=Kicks, parameterName=IsKicks}
  }
  MaterialExpressionStaticSwitchParameter_3 {
   class=MaterialExpressionStaticSwitchParameter
   inputs [
    True {from=MaterialExpressionSetMaterialAttributes_0, output=""}
    False {from=MaterialExpressionStaticSwitchParameter_2, output=""}
   ]
   properties {group=SaltyBrim, parameterName=UseSaltyBrimEmote}
  }
  MaterialExpressionSubtract_0 {
   class=MaterialExpressionSubtract
   inputs [
    A {from=MaterialExpressionComponentMask_0}
    B {from=""}
   ]
   properties {constB=0.0001}
  }
  MaterialExpressionSubtract_1 {
   class=MaterialExpressionSubtract
   inputs [
    A {from=MaterialExpressionFrac_1}
    B {from=MaterialExpressionAppendVector_6}
   ]
  }
  MaterialExpressionSubtract_2 {
   class=MaterialExpressionSubtract
   inputs [
    A {from=MaterialExpressionFrac_2}
    B {from=""}
   ]
   properties {constB=0.5}
  }
  MaterialExpressionSubtract_3 {
   class=MaterialExpressionSubtract
   inputs [
    A {from=MaterialExpressionComponentMask_2}
    B {from=MaterialExpressionMultiply_1}
   ]
  }
  MaterialExpressionSubtract_4 {
   class=MaterialExpressionSubtract
   inputs [
    A {from=MaterialExpressionTime_0}
    B {from=MaterialExpressionAdd_1}
   ]
  }
  MaterialExpressionSubtract_6 {
   class=MaterialExpressionSubtract
   inputs [
    A {from=MaterialExpressionLinearInterpolate_19}
    B {from=""}
   ]
   properties {constB=0.5}
  }
  MaterialExpressionSwitch_0 {
   class=MaterialExpressionSwitch
   inputs [
    SwitchValue {from=MaterialExpressionScalarParameter_1, output=""}
    Default {from="", output=""}
    None {from=MaterialExpressionVectorParameter_0, output=0 - Mouth Neutral}
    None {from=MaterialExpressionVectorParameter_0, output=0 - Mouth Neutral}
    None {from=MaterialExpressionVectorParameter_0, output=0 - Mouth Neutral}
    None {from=MaterialExpressionVectorParameter_0, output=0 - Mouth Neutral}
    None {from=MaterialExpressionVectorParameter_1, output=4 - Jaw Open Frown}
    None {from=MaterialExpressionVectorParameter_1, output=4 - Jaw Open Frown}
    None {from=MaterialExpressionVectorParameter_1, output=4 - Jaw Open Frown}
    None {from=MaterialExpressionVectorParameter_1, output=4 - Jaw Open Frown}
    None {from=MaterialExpressionVectorParameter_2, output=8 - Smile L Wide}
    None {from=MaterialExpressionVectorParameter_2, output=8 - Smile L Wide}
    None {from=MaterialExpressionVectorParameter_2, output=8 - Smile L Wide}
    None {from=MaterialExpressionVectorParameter_2, output=8 - Smile L Wide}
    None {from=MaterialExpressionVectorParameter_3, output=12 - Frown}
    None {from=MaterialExpressionVectorParameter_3, output=12 - Frown}
    None {from=MaterialExpressionVectorParameter_3, output=12 - Frown}
   ]
   properties {inputs [
     {
      inputName=None
      input {
       expression=&self:MaterialExpressionVectorParameter_0
       inputName=None
       mask=1
       maskA=0
       maskB=0
       maskG=0
       maskR=1
       outputIndex=1
      }
     }
     {
      inputName=None
      input {
       expression=&self:MaterialExpressionVectorParameter_0
       inputName=None
       mask=1
       maskA=0
       maskB=0
       maskG=1
       maskR=0
       outputIndex=2
      }
     }
     {
      inputName=None
      input {
       expression=&self:MaterialExpressionVectorParameter_0
       inputName=None
       mask=1
       maskA=0
       maskB=1
       maskG=0
       maskR=0
       outputIndex=3
      }
     }
     {
      inputName=None
      input {
       expression=&self:MaterialExpressionVectorParameter_0
       inputName=None
       mask=1
       maskA=1
       maskB=0
       maskG=0
       maskR=0
       outputIndex=4
      }
     }
     {
      inputName=None
      input {
       expression=&self:MaterialExpressionVectorParameter_1
       inputName=None
       mask=1
       maskA=0
       maskB=0
       maskG=0
       maskR=1
       outputIndex=1
      }
     }
     {
      inputName=None
      input {
       expression=&self:MaterialExpressionVectorParameter_1
       inputName=None
       mask=1
       maskA=0
       maskB=0
       maskG=1
       maskR=0
       outputIndex=2
      }
     }
     {
      inputName=None
      input {
       expression=&self:MaterialExpressionVectorParameter_1
       inputName=None
       mask=1
       maskA=0
       maskB=1
       maskG=0
       maskR=0
       outputIndex=3
      }
     }
     {
      inputName=None
      input {
       expression=&self:MaterialExpressionVectorParameter_1
       inputName=None
       mask=1
       maskA=1
       maskB=0
       maskG=0
       maskR=0
       outputIndex=4
      }
     }
     {
      inputName=None
      input {
       expression=&self:MaterialExpressionVectorParameter_2
       inputName=None
       mask=1
       maskA=0
       maskB=0
       maskG=0
       maskR=1
       outputIndex=1
      }
     }
     {
      inputName=None
      input {
       expression=&self:MaterialExpressionVectorParameter_2
       inputName=None
       mask=1
       maskA=0
       maskB=0
       maskG=1
       maskR=0
       outputIndex=2
      }
     }
     {
      inputName=None
      input {
       expression=&self:MaterialExpressionVectorParameter_2
       inputName=None
       mask=1
       maskA=0
       maskB=1
       maskG=0
       maskR=0
       outputIndex=3
      }
     }
     {
      inputName=None
      input {
       expression=&self:MaterialExpressionVectorParameter_2
       inputName=None
       mask=1
       maskA=1
       maskB=0
       maskG=0
       maskR=0
       outputIndex=4
      }
     }
     {
      inputName=None
      input {
       expression=&self:MaterialExpressionVectorParameter_3
       inputName=None
       mask=1
       maskA=0
       maskB=0
       maskG=0
       maskR=1
       outputIndex=1
      }
     }
     {
      inputName=None
      input {
       expression=&self:MaterialExpressionVectorParameter_3
       inputName=None
       mask=1
       maskA=0
       maskB=0
       maskG=1
       maskR=0
       outputIndex=2
      }
     }
     {
      inputName=None
      input {
       expression=&self:MaterialExpressionVectorParameter_3
       inputName=None
       mask=1
       maskA=0
       maskB=1
       maskG=0
       maskR=0
       outputIndex=3
      }
     }
    ]}
  }
  MaterialExpressionTemporalResponsivenessOutput_2 {class=MaterialExpressionTemporalResponsivenessOutput, inputs [None {from=MaterialExpressionMax_0, output=""}], properties {}}
  MaterialExpressionTextureCoordinate_0 {class=MaterialExpressionTextureCoordinate, properties {coordinateIndex=1}}
  MaterialExpressionTextureCoordinate_1 {class=MaterialExpressionTextureCoordinate, properties {coordinateIndex=1}}
  MaterialExpressionTextureCoordinate_2 {class=MaterialExpressionTextureCoordinate, properties {coordinateIndex=1}}
  MaterialExpressionTextureCoordinate_9 {class=MaterialExpressionTextureCoordinate, properties {coordinateIndex=1}}
  MaterialExpressionTextureSampleParameter2D_0 {
   class=MaterialExpressionTextureSampleParameter2D
   inputs [
    UVs {from="", output=""}
    Apply View MipBias {from="", output=""}
   ]
   properties {
    group=01_EyeAnimation_Parameters
    parameterName=EyeMask
    samplerSource=SSM_Wrap_WorldGroupSettings
    samplerType=SAMPLERTYPE_Masks
    sortPriority=-32
    texture=&T_WhiteMask
   }
  }
  MaterialExpressionTime_0 {class=MaterialExpressionTime}
  MaterialExpressionTransformPosition_0 {
   class=MaterialExpressionTransformPosition
   inputs [None {from=MaterialExpressionVectorParameter_4, output=RGB}]
   properties {transformType=TRANSFORMPOSSOURCE_TranslatedWorld}
  }
  MaterialExpressionTransform_18 {
   class=MaterialExpressionTransform
   inputs [None {from=MaterialExpressionNormalize_1}]
   properties {transformSourceType=TRANSFORMSOURCE_World, transformType=TRANSFORM_Local}
  }
  MaterialExpressionVectorParameter_0 {
   class=MaterialExpressionVectorParameter
   properties {parameterName=MiddleMouth_0-3, channelNames {a=3 - Jaw Open R Smile, b=2 - Jaw Open L Smile, g=1 - Mouth Pucker, r=0 - Mouth Neutral}}
  }
  MaterialExpressionVectorParameter_1 {
   class=MaterialExpressionVectorParameter
   properties {parameterName=MiddleMouth_4-7, channelNames {a=7 - Jaw Open, b=6 - Jaw Open R Snear, g=5 -Jaw Open L Snear, r=4 - Jaw Open Frown}}
  }
  MaterialExpressionVectorParameter_2 {
   class=MaterialExpressionVectorParameter
   properties {parameterName=MiddleMouth_8-11, channelNames {a=11 - Smile, b=10 - Frown Wide, g=9 - Smile R Wide, r=8 - Smile L Wide}}
  }
  MaterialExpressionVectorParameter_3 {
   class=MaterialExpressionVectorParameter
   properties {parameterName=MiddleMouth12-15, sortPriority=33, channelNames {a=Unused, b=14 - Snear R, g=13 - Snear L, r=12 - Frown}}
  }
  MaterialExpressionVectorParameter_4 {class=MaterialExpressionVectorParameter, properties {parameterName=Head Position}}
  MaterialExpressionVectorParameter_8 {class=MaterialExpressionVectorParameter, properties {parameterName=HeadForwardVector, defaultValue=<rgba 0 1 0 0>}}
  MaterialExpressionVectorParameter_9 {class=MaterialExpressionVectorParameter, properties {parameterName=HeadRightVector, defaultValue=<rgba 1 0 0 0>}}
  MaterialExpressionVertexColor_0 {class=MaterialExpressionVertexColor, inputs [], properties {}}
 }
 properties {
  bAutomaticallySetUsageInEditor=false
  bUseMaterialAttributes=true
  bUsedWithClothing=true
  bUsedWithMeshDeformer=true
  bUsedWithMorphTargets=true
  bUsedWithNiagaraMeshParticles=true
  bUsedWithSkeletalMesh=true
  bUsedWithStaticLighting=true
  pixelDepthOffsetMode=PDOM_Legacy
  previewMesh=/BRCosmetics/Characters/Player/Male/Medium/Bodies/M_MED_SaltyBrim/Meshes/Parts/M_MED_SaltyBrim_FaceAcc
  shadingModel=MSM_Cloth
  usedShadingModels=Cloth
 }
}
